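# Previs Level Sequence for the NightNight event (Part 3, Set 4, rocket 04) that flies a separated rocket out of a rift.
# A rift-exit effect spawns shortly before the rocket appears, which travels via an animated CameraRig_Rail position and attach tracks, with a rolling parent mesh.
# Late in flight, passby audio is disabled, blue engines switch to afterburner, and the stage-two mesh scales to zero before the rocket despawns.
# A constant-strength camera shake component rides on the rocket.
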LevelSequence /Figment_S10/Athena/Prototype/Blueprints/NightNight/Sequences/RocketTests_P3_3/Set4/NightNight_Animation_PREVIS_Part3_Rocket_04
bindings [
 P_Geode_2_BlueEngines {
  id=#1
  parent_id=#2
  possessed_class=ParticleSystemComponent
  tracks [{
    type=MovieSceneParticleTrack
    display_name=Particle System
    sections [{
      type=MovieSceneParticleSection
      channels [None {type=MovieSceneScriptingParticleChannel, keys [[1259, "<ParticleKey.ACTIVATE: 0>"], [2585, "<ParticleKey.DEACTIVATE: 1>"]]}]
     }]
   }]
 }
 Rocket_01_Sep {
  id=#2
  possessed_class=BP_Geode_NightNight_Separated_C
  tracks [
   {
    type=MovieSceneSpawnTrack
    sections [{type=MovieSceneSpawnSection, channels [None {type=MovieSceneScriptingBoolChannel, default=true, keys [[0,false],[1259,true],[2808,false]]}]}]
   }
   {
    type=MovieScene3DTransformTrack
    sections [{
      type=MovieScene3DTransformSection
      end_frame=2808
      start_frame=1019
      channels [
       Location.X {type=MovieSceneScriptingDoubleChannel, default=0}
       Location.Y {type=MovieSceneScriptingDoubleChannel, default=0}
       Location.Z {type=MovieSceneScriptingDoubleChannel, default=0}
       Rotation.X {type=MovieSceneScriptingDoubleChannel, default=1.0998}
       Rotation.Y {type=MovieSceneScriptingDoubleChannel, default=-89.5244}
       Rotation.Z {type=MovieSceneScriptingDoubleChannel, default=-1.0998}
      ]
     }]
   }
   {type=MovieScene3DAttachTrack, sections [{type=MovieScene3DAttachSection, end_frame=3520, start_frame=1019}]}
   {
    type=MovieSceneVisibilityTrack
    sections [{
      type=MovieSceneVisibilitySection
      channels [None {type=MovieSceneScriptingBoolChannel, default=false, keys [[0,false],[1259,true],[2808,false]]}]
     }]
   }
   {
    type=MovieSceneBoolTrack
    display_name=Disable Passby Audio
    sections [{type=MovieSceneBoolSection, channels [None {type=MovieSceneScriptingBoolChannel, default=false, keys [[1259,false],[2466,true]]}]}]
   }
  ]
 }
 Rocket_01_Path {
  id=#3
  possessed_class=CameraRig_Rail
  tracks [
   {
    type=MovieSceneSpawnTrack
    sections [{type=MovieSceneSpawnSection, channels [None {type=MovieSceneScriptingBoolChannel, default=false, keys [[0,false],[1019,true]]}]}]
   }
   {
    type=MovieScene3DTransformTrack
    sections [{
      type=MovieScene3DTransformSection
      channels [
       Location.X {type=MovieSceneScriptingDoubleChannel, default=0.2928, const [0.2928,1]}
       Location.Y {type=MovieSceneScriptingDoubleChannel, default=167.0334, const [167.0334,1]}
       Location.Z {type=MovieSceneScriptingDoubleChannel, default=72754.4375, const [72754.4375,1]}
       Rotation.X {type=MovieSceneScriptingDoubleChannel, default=90, const [90,1]}
       Rotation.Y {type=MovieSceneScriptingDoubleChannel, default=0, const [0,1]}
       Rotation.Z {type=MovieSceneScriptingDoubleChannel, default=0, const [0,1]}
       Scale.X {type=MovieSceneScriptingDoubleChannel, default=1, const [1,1]}
       Scale.Y {type=MovieSceneScriptingDoubleChannel, default=1, const [1,1]}
       Scale.Z {type=MovieSceneScriptingDoubleChannel, default=1, const [1,1]}
      ]
     }]
   }
   {type=MovieScene3DAttachTrack, sections [{type=MovieScene3DAttachSection, end_frame=3520, start_frame=1019}]}
   {
    type=MovieSceneFloatTrack
    display_name=Current Position on Rail
    sections [{
      type=MovieSceneFloatSection
      channels [None {type=MovieSceneScriptingFloatChannel, default=1, keys [[1259,0],[2575,0.7924],[2628,0.8435],[2718,1]]}]
     }]
   }
  ]
 }
 BP_NightNight_Generic_Rift_Exit {
  id=#4
  possessed_class=BP_NightNight_Generic_RiftExit_Quick_C
  tracks [
   {
    type=MovieSceneSpawnTrack
    sections [{
      type=MovieSceneSpawnSection
      channels [None {type=MovieSceneScriptingBoolChannel, default=false, keys [[990,false],[1019,true],[1562,false]]}]
     }]
   }
   {
    type=MovieScene3DTransformTrack
    sections [{
      type=MovieScene3DTransformSection
      channels [
       Location.X {type=MovieSceneScriptingDoubleChannel, default=101591.2812}
       Location.Y {type=MovieSceneScriptingDoubleChannel, default=-93967.9141}
       Location.Z {type=MovieSceneScriptingDoubleChannel, default=13774.5283}
       Rotation.X {type=MovieSceneScriptingDoubleChannel, default=179.1864}
       Rotation.Y {type=MovieSceneScriptingDoubleChannel, default=4.0941}
       Rotation.Z {type=MovieSceneScriptingDoubleChannel, default=175.7273}
       Scale.X {type=MovieSceneScriptingDoubleChannel, default=1}
       Scale.Y {type=MovieSceneScriptingDoubleChannel, default=1}
       Scale.Z {type=MovieSceneScriptingDoubleChannel, default=1}
      ]
     }]
   }
   {type=MovieScene3DAttachTrack, sections []}
  ]
 }
 MissileSilo_Stage2_Main {
  id=#5
  parent_id=#2
  possessed_class=StaticMeshComponent
  tracks [{
    type=MovieScene3DTransformTrack
    sections [{
      type=MovieScene3DTransformSection
      channels [
       Scale.X {type=MovieSceneScriptingDoubleChannel, default=0, keys [[2639,1],[2718,0]]}
       Scale.Y {type=MovieSceneScriptingDoubleChannel, default=0, keys [[2639,1],[2718,0]]}
       Scale.Z {type=MovieSceneScriptingDoubleChannel, default=0, keys [[2639,1],[2718,0]]}
      ]
     }]
   }]
 }
 RocketRoot {
  id=#6
  possessed_class=FortStaticMeshActor
  tracks [
   {
    type=MovieScene3DTransformTrack
    sections [{
      type=MovieScene3DTransformSection
      channels [
       Location.X {type=MovieSceneScriptingDoubleChannel, default=30047.4805}
       Location.Y {type=MovieSceneScriptingDoubleChannel, default=-18197.7656}
       Location.Z {type=MovieSceneScriptingDoubleChannel, default=192048.0938}
       Rotation.X {type=MovieSceneScriptingDoubleChannel, default=180}
       Rotation.Y {type=MovieSceneScriptingDoubleChannel, default=0}
       Rotation.Z {type=MovieSceneScriptingDoubleChannel, default=135}
       Scale.X {type=MovieSceneScriptingDoubleChannel, default=1}
       Scale.Y {type=MovieSceneScriptingDoubleChannel, default=1}
       Scale.Z {type=MovieSceneScriptingDoubleChannel, default=1}
      ]
     }]
   }
   {type=MovieSceneSpawnTrack, sections [{type=MovieSceneSpawnSection, channels [None {type=MovieSceneScriptingBoolChannel, default=true}]}]}
  ]
 }
 P_NightNight_Afterburner {
  id=#7
  parent_id=#2
  possessed_class=ParticleSystemComponent
  tracks [{
    type=MovieSceneParticleTrack
    display_name=Particle System
    sections [{
      type=MovieSceneParticleSection
      channels [None {type=MovieSceneScriptingParticleChannel, keys [[1259, "<ParticleKey.DEACTIVATE: 1>"], [2585, "<ParticleKey.ACTIVATE: 0>"]]}]
     }]
   }]
 }
 RiftExit {
  id=#8
  parent_id=#4
  possessed_class=ParticleSystemComponent
  tracks [{
    type=MovieSceneParticleTrack
    display_name=Particle System
    sections [{type=MovieSceneParticleSection, channels [None {type=MovieSceneScriptingParticleChannel, const ["<ParticleKey.TRIGGER: 2>", 1]}]}]
   }]
 }
 Cylinder {
  id=#9
  possessed_class=FortStaticMeshActor
  tracks [
   {
    type=MovieScene3DTransformTrack
    sections [{
      type=MovieScene3DTransformSection
      channels [
       Location.X {type=MovieSceneScriptingDoubleChannel, default=0}
       Location.Y {type=MovieSceneScriptingDoubleChannel, default=0}
       Location.Z {type=MovieSceneScriptingDoubleChannel, default=0}
       Rotation.X {type=MovieSceneScriptingDoubleChannel, default=0}
       Rotation.Y {type=MovieSceneScriptingDoubleChannel, default=0}
       Rotation.Z {type=MovieSceneScriptingDoubleChannel, default=0}
       Scale.X {type=MovieSceneScriptingDoubleChannel, default=444}
       Scale.Y {type=MovieSceneScriptingDoubleChannel, default=444}
       Scale.Z {type=MovieSceneScriptingDoubleChannel, default=99999}
      ]
     }]
   }
   {type=MovieSceneSpawnTrack, sections [{type=MovieSceneSpawnSection, channels [None {type=MovieSceneScriptingBoolChannel, default=true}]}]}
   {type=MovieScene3DAttachTrack, sections [{type=MovieScene3DAttachSection, end_frame=2642, start_frame=0}]}
  ]
 }
 Rocket_01_Parent {
  id=#10
  possessed_class=FortStaticMeshActor
  tracks [
   {
    type=MovieScene3DTransformTrack
    sections [{
      type=MovieScene3DTransformSection
      channels [
       Location.X {type=MovieSceneScriptingDoubleChannel, default=0, const [0,1]}
       Location.Y {type=MovieSceneScriptingDoubleChannel, default=0, const [0,1]}
       Location.Z {type=MovieSceneScriptingDoubleChannel, default=0, const [0,1]}
       Rotation.X {type=MovieSceneScriptingDoubleChannel, default=0, keys [[1259,0],[1442,119.0061],[1559,-69.2726]]}
       Rotation.Y {type=MovieSceneScriptingDoubleChannel, default=0, keys [[1259,0],[1442,0.0013],[1559,0.0019]]}
       Rotation.Z {type=MovieSceneScriptingDoubleChannel, default=0, keys [[1259,0],[1442,0.0009],[1559,0.0007]]}
       Scale.X {type=MovieSceneScriptingDoubleChannel, default=1, const [1,1]}
       Scale.Y {type=MovieSceneScriptingDoubleChannel, default=1, const [1,1]}
       Scale.Z {type=MovieSceneScriptingDoubleChannel, default=1, const [1,1]}
      ]
     }]
   }
   {type=MovieSceneSpawnTrack, sections [{type=MovieSceneSpawnSection, channels [None {type=MovieSceneScriptingBoolChannel, default=true}]}]}
   {type=MovieScene3DAttachTrack, sections [{type=MovieScene3DAttachSection, end_frame=3520, start_frame=1019}]}
  ]
 }
 NightNight_SmallerRocket_CameraShakeComponent {
  id=#11
  parent_id=#2
  possessed_class=NightNight_SmallerRocket_CameraShakeComponent_C
  tracks [{
    type=MovieSceneFloatTrack
    display_name=Shake Strength
    sections [{type=MovieSceneFloatSection, channels [None {type=MovieSceneScriptingFloatChannel, default=0.5, const [0.5,1]}]}]
   }]
 }
]
display_rate {denominator=1, numerator=60}
tick_resolution {denominator=1, numerator=24000}
tracks []
